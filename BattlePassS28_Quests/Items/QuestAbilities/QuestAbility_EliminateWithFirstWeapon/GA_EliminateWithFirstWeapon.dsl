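; Quest ability that broadcasts EliminateEnemyWithFirstWeapon when the player eliminates an enemy while holding their first looted weapon.
; Pickup events record the first valid weapon's tracker GUID, and an aircraft-exit listener calls a reset and clears it.
; Kill events pass through CheckTags, which excludes listed NPC tags and requires DBNO tags outside solo playlists.
; Each new target is recorded once, then the current weapon's GUID is compared to the stored one.
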(event Ability|EventActivateAbility
  (bind _returnvalue (Ability|GetAvatarActorfromActorInfo))
  (Utilities|FlowControl|Sequence
    (:then_0
      (bind (_asynctaskproxy _payload) (Ability|Async|WaitGameplayEventToActor _returnvalue "(TagName=\"Event.CollectedPickup\")")
        (:then
          (Ability|Async|WaitGameplayEventToActor (Ability|GetAvatarActorfromActorInfo) "(TagName=\"Event.Aircraft.Exited\")"
            (:EventReceived
              (CallFunction|ResetFirstWeaponGUID)
              (Variables|Default|SetFirstWeaponLooted))))
        (:EventReceived
          (bind _asfort_pickup_athena (Utilities|Casting|CastToFortPickupAthena (.Target (Utilities|Struct|BreakGameplayEventData _payload))))
          (if (GameplayTags|HasAnyMatchingGameplayTags _asfort_pickup_athena (Variables|Default|GetValidWeaponTags))
            (Utilities|FlowControl|DoOnce false
              (:Completed
                (Variables|Default|SetFirstWeaponLooted (Pickup|GetTrackerGuid _asfort_pickup_athena))
                (Development|PrintString "LOOTED FIRST GUN!" false false "(R=1.000000,G=0.248857,B=0.000000,A=1.000000)" 5.0)))))))
    (:then_1
      (bind (_asynctaskproxy_2 _payload_2) (Ability|Async|WaitGameplayEventToActor _returnvalue "(TagName=\"Event.Damage.Killed\")")
        (:EventReceived
          (bind _output_get (Variables|Default|SetIncomingTarget (.Target (Utilities|Struct|BreakGameplayEventData _payload_2))))
          (bind _success (CallFunction|CheckTags _output_get))
          (if _success
            (if (not (Utilities|Array|ContainsItem (Variables|Default|GetEliminatedTargets) (Variables|Default|GetIncomingTarget)))
              (Utilities|Array|AddUnique (Variables|Default|GetEliminatedTargets) (Variables|Default|GetIncomingTarget))
              (if (== (FortWeapon|GetTrackerGuid (Weapon|GetCurrentWeapon (Game|Player|GetPlayerPawn (Fortnite|GetFortPlayerControllerfromActor (Ability|GetAvatarActorfromActorInfo))))) (Variables|Default|GetFirstWeaponLooted))
                (CustomVerb|BroadcastCustomVerbMessage (CustomVerb|MakeCustomVerbMessage (Ability|GetAvatarActorfromActorInfo) "(TagName=\"CustomVerb.EliminateEnemyWithFirstWeapon\")"))))))))))

(event Custom|ResetFirstWeaponGUID
  (Utilities|FlowControl|DoOnce false
    (:Completed
      (Variables|Default|SetFirstWeaponLooted (Pickup|GetTrackerGuid (stale Ability|EventActivateAbility _asfort_pickup_athena)))
      (Development|PrintString "LOOTED FIRST GUN!" false false "(R=1.000000,G=0.248857,B=0.000000,A=1.000000)" 5.0))))

(event Ability|EventOnEndAbility
  (params (bWasCancelled :type Boolean)))

(fn CheckTags
  (params (InActor :type Actor%20Object%20Reference))
  (bind (_outgamecontextgameplaytags _returnvalue) (GameContext|GetGameContextGameplayTags))
  (bind _returnvalue_2 (GameplayTags|HasAnyMatchingGameplayTags(Message) (Variables|GetInActor) (GameplayTags|MakeLiteralGameplayTagContainer "(GameplayTags=((TagName=\"Homebase.Class\")))")))
  (if _returnvalue_2
    (bind _returnvalue_3 (GameplayTags|HasAnyMatchingGameplayTags(Message) (Variables|GetInActor) (GameplayTags|MakeGameplayTagContainerfromArray (Utilities|Array|MakeArray "(TagName=\"Athena.Faction.Generic.Hostile\")" "(TagName=\"Pawn.Athena.NPC.Duel.IsBeingChallenged\")" "(TagName=\"AISpawnerData.Type.Boss\")" "(TagName=\"AISpawnerData.Type.MMObj.BunkerGuard.Boss\")" "(TagName=\"AISpawnerData.Type.MMObj.BunkerGuard.Henchman\")" "(TagName=\"Athena.Faction.Hopscotch.BaseGuard\")" "(TagName=\"AISpawnerData.Type.MMObj.RadioTowerGuard\")" "(TagName=\"AISpawnerData.Type.MMObj.RadioTowerGuard.Boss\")" "(TagName=\"AISpawnerData.Type.MMObj.RadioTowerGuard.Henchman\")" "(TagName=\"AISpawnerData.Type.Hopscotch\")" "(TagName=\"Pawn.Athena.NPC.Labrador\")" "(TagName=\"Pawn.Athena.NPC.Wildlife\")"))))
    (if (not _returnvalue_3)
      (if _returnvalue
        (bind _output_get (Variables|Default|SetInSoloMatch (or (Utilities|Array|ContainsItem (GameplayTags|BreakGameplayTagContainer _outgamecontextgameplaytags) "(TagName=\"Athena.Playlist.NoBuildBR.Solo\")") (Utilities|Array|ContainsItem (GameplayTags|BreakGameplayTagContainer _outgamecontextgameplaytags) "(TagName=\"Athena.Playlist.Solo\")"))))
        (if _output_get
          (return true)
          (else
            (bind _returnvalue_4 (GameplayTags|HasAnyMatchingGameplayTags(Message) (Variables|GetInActor) (GameplayTags|MakeGameplayTagContainerfromArray (Utilities|Array|MakeArray "(TagName=\"Gameplay.Status.DBNO\")" "(TagName=\"Gameplay.Status.DBNOBeingCarried\")" "(TagName=\"Gameplay.Status.DBNOBeingDropped\")" "(TagName=\"Gameplay.Status.DBNOBeingRevived\")" "(TagName=\"Gameplay.Status.DBNOBlocked\")" "(TagName=\"Gameplay.ImprovedDBNO\")"))))
            (if _returnvalue_4
              (return true)
              (else
                (return false)))))
        (else
          (return false)))
      (else
        (return false)))
    (else
      (return false))))
